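# Translucent, two-sided unlit material for the Radio Tower's radar cone in mid-match gameplay.
# Two cone-texture samples, with time-scrolled UV math alongside, combine with random-texture variation, a view-angle fade at grazing edges and vertex-color masking, plus a tight vertex-color highlight.
# An Activation Complete custom primitive data value blends cyan toward green, tinted red where the first cone sample exceeds the second.
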Material /MidMatchRadioTowerGameplay/Art/Materials/Radar/M_Gameplay_RadioTower_RadarCone
graph {
 properties {
  blendMode=BLEND_Translucent
  numCustomizedUVs=4
  pixelDepthOffsetMode=PDOM_Legacy
  shadingModel=MSM_Unlit
  twoSided=true
  usedShadingModels=Unlit
 }
 nodes [
  n1 = TextureSample(UVs=TextureCoordinate() {coordinateIndex=1}) {samplerType=SAMPLERTYPE_LinearGrayscale, texture=&/T_Gameplay_RadioTower_Radar_Cone}
  n2 = TextureSample(UVs=TextureCoordinate() {coordinateIndex=2}) {samplerType=SAMPLERTYPE_LinearGrayscale, texture=&T_Gameplay_RadioTower_Radar_Cone}
  n3 = TextureSample(UVs=TextureCoordinate() {coordinateIndex=3}) {samplerType=SAMPLERTYPE_LinearGrayscale, texture=&/T_Gameplay_RadioTower_Radar_Random_A}
  n4 = CameraVectorWS()
  n5 = VertexNormalWS()
  n6 = Multiply(A=OneMinus(None=Multiply(A=@n3.R) {constB=0.4}), B=Abs(None=DotProduct(A=@n4, B=@n5)))
  n7 = Multiply(A=Multiply(A=Add(A=@n1.R, B=@n2.R)) {constB=1.2}, B=Multiply(A=@n6, B=VertexColor().R))
  n8 = Power(Base=VertexColor().R) {constExponent=16}
  Add(A=Multiply(A=@n7, B=@n7), B=Multiply(A=Multiply(A=@n6, B=@n8)) {constB=0.4})
  n9 = TextureCoordinate()
  n10 = ComponentMask(None=@n9) {r=true}
  n11 = Time()
  n12 = ComponentMask(None=@n9) {g=true}
  AppendVector(
   A=Subtract(A=Multiply(A=@n10) {constB=1.6}, B=Multiply(A=@n11) {constB=0.17})
   B=Add(A=Multiply(A=@n12) {constB=0.6}, B=Multiply(A=@n11) {constB=0.2})
  )
  n13 = Time()
  n14 = Multiply(A=@n12)
  n15 = Multiply(A=@n13) {constB=0.1}
  AppendVector(A=Subtract(A=Multiply(A=@n10) {constB=1.4}, B=Multiply(A=@n13) {constB=0.15}), B=Add(A=@n14, B=@n15))
  n16 = Saturate(None=Subtract(A=@n1.R, B=@n2.R))
  LinearInterpolate(
   A=LinearInterpolate(
    A=Constant3Vector() {constant=<rgba 0.2708 1.5619 2 0>}
    B=Constant3Vector() {constant=<rgba 0.1049 1.5 0 1>}
    Alpha=ScalarParameter() {bUseCustomPrimitiveData=true, parameterName=Activation Complete, primitiveDataIndex=2}
   )
   B=Constant3Vector() {constant=<rgba 1 0 0.1064 0>}
   Alpha=@n16
  )
  Multiply(A=@n16, B=@n16)
  Multiply(A=Time()) {constB=0.2}
  Subtract(A=@n14, B=@n15)
  TextureCoordinate()
 ]
}
