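# MF_RadioTower_Random is a material function for the radio tower art that outputs a pseudo-random 0–1 scalar per object, derived from its pivot location.
# It takes the absolute object pivot position, scales it by a constant vector, then sums the components and keeps the fractional part.
# The sum adds X twice plus Z and never reads Y, so objects differing only in Y get the same value.
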MaterialFunction /MidMatchRadioTowerGameplay/Art/Materials/MF_RadioTower_Random
graph {nodes [
  n1 = MaterialFunctionCall() {materialFunction=&ObjectPivotPoint}
  n2 = MaterialFunctionCall(Float3=Multiply(A=Abs(None=@n1.Object Pivot Location), B=Constant3Vector() {constant=<rgba 0.328 0.128 0.36 0>})) {materialFunction=&BreakOutFloat3Components}
  n3 = Add(A=@n2.R, B=@n2.R)
  FunctionOutput(None=Frac(None=Add(A=@n3, B=@n2.B))) {outputName=Random Scalar}
 ]}
