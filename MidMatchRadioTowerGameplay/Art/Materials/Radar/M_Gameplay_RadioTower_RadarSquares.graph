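# Translucent unlit material for the radio tower radar, drawing animated glitchy squares as looping sprites moved entirely by vertex offset.
# Randomly seeded quads are expanded toward the view, travel along the local axis with per-loop scatter and a circling tip, and shrink over each cycle.
# Noise textures jitter three square masks into RGB-split fringes.
# Activation Complete, from custom primitive data, slows travel, makes it ping-pong and shifts cyan to green.
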Material /MidMatchRadioTowerGameplay/Art/Materials/Radar/M_Gameplay_RadioTower_RadarSquares
graph {
 properties {blendMode=BLEND_Translucent, pixelDepthOffsetMode=PDOM_Legacy, shadingModel=MSM_Unlit, twoSided=true, usedShadingModels=Unlit}
 nodes [
  n1 = AppendVector(A=Add(A=TextureCoordinate(), B=Constant() {r=-0.5}), B=Constant())
  n2 = Time() {period=1}
  n3 = ComponentMask(None=PreSkinnedPosition()) {b=true}
  n4 = Frac(None=Add(
    A=Multiply(A=@n3) {constB=2.8805}
    B=MaterialFunctionCall() {materialFunction=/MidMatchRadioTowerGameplay/Art/Materials/MF_RadioTower_Random}.Random Scalar
   ))
  n5 = TextureSample(UVs=Multiply(A=Add(A=@n2, B=@n4)) {constB=0.06}) {samplerType=SAMPLERTYPE_Normal, texture=&/T_Gameplay_RadioTower_Radar_Random}
  n6 = ScalarParameter() {defaultValue=60, parameterName=SpriteWidth}
  n7 = ScalarParameter() {defaultValue=30, parameterName=SpriteHeightMin}
  n8 = ScalarParameter() {defaultValue=70, parameterName=SpriteHeightMax}
  n9 = Multiply(
   A=Add(A=Multiply(A=@n5.RGB, B=@n6) {constB=100}, B=LinearInterpolate(A=@n7, B=@n8, Alpha=@n4) {constA=20, constB=40})
   B=MaterialFunctionCall() {materialFunction=&ObjectScale}.Scale X
  )
  n10 = VertexNormalWS()
  n11 = Multiply(A=Multiply(A=ComponentMask(None=@n1) {r=true}, B=ComponentMask(None=@n9) {r=true}) {constB=2}, B=@n10)
  n12 = Normalize(VectorInput=Subtract(A=WorldPosition().XYZ, B=CameraPositionWS()))
  n13 = Multiply(
   A=Multiply(A=ComponentMask(None=@n1) {g=true}, B=ComponentMask(None=@n9) {g=true}) {constB=2}
   B=Normalize(VectorInput=CrossProduct(A=@n10, B=@n12))
  ) {constB=2}
  n14 = ScalarParameter() {bUseCustomPrimitiveData=true, parameterName=Activation Complete, primitiveDataIndex=2}
  n15 = Add(
   A=Multiply(A=@n3) {constB=4.7908}
   B=MaterialFunctionCall() {materialFunction=&MF_RadioTower_Random}.Random Scalar
  )
  n16 = Multiply(
   A=Multiply(A=Time() {period=1}, B=LinearInterpolate(Alpha=@n14) {constA=-1, constB=-0.25}) {constB=-0.25}
   B=LinearInterpolate(Alpha=Frac(None=@n15)) {constA=0.2, constB=0.8}
  ) {constB=0.5}
  n17 = Frac(None=@n16)
  n18 = Add(A=@n17) {constB=-0.5}
  n19 = LinearInterpolate(
   A=@n17
   B=Multiply(A=Abs(None=@n18)) {constB=2}
   Alpha=ScalarParameter() {bUseCustomPrimitiveData=true, parameterName=Activation Complete, primitiveDataIndex=2}
  )
  n20 = OneMinus(None=@n19)
  n21 = Floor(None=Abs(None=@n16))
  n22 = AppendVector(A=Frac(None=Multiply(A=@n21) {constB=16.4222}), B=Frac(None=Multiply(A=@n21) {constB=10.5061}))
  n23 = OneMinus(None=Multiply(A=@n19) {constB=0.3})
  n24 = ScalarParameter() {defaultValue=200, parameterName=Width}
  n25 = ScalarParameter() {defaultValue=1000, parameterName=Length}
  n26 = PreSkinnedPosition()
  n27 = AppendVector(
   A=Multiply(A=Add(A=@n22) {constB=-0.5}, B=Multiply(A=@n23, B=@n24)) {constB=10}
   B=Subtract(A=Multiply(A=@n19, B=@n25) {constB=40}, B=ComponentMask(None=@n26) {b=true})
  )
  n28 = Multiply(A=Add(A=Multiply(A=@n19) {constB=2}, B=Time())) {constB=0.5}
  MaterialExpressionAppend3Vector_1 = MaterialXAppend3Vector!(A=Sine(None=@n28), B=Cosine(None=@n28), C=Constant())
  n29 = Multiply(
   A=Multiply(A=@n19, B=@n19)
   B=Multiply(A=@MaterialExpressionAppend3Vector_1, B=ScalarParameter() {defaultValue=50, parameterName=CircleTip}) {constB=100}
  )
  Add(
   A=Multiply(A=Add(A=@n11, B=@n13), B=SquareRoot(None=@n20))
   B=Transform(None=Add(A=@n27, B=@n29)) {transformSourceType=TRANSFORMSOURCE_Local}
  )
  n30 = Add(A=TextureCoordinate()) {constB=-0.5}
  n31 = Add(
   A=Multiply(A=@n3) {constB=0.8862}
   B=MaterialFunctionCall() {materialFunction=&MF_RadioTower_Random}.Random Scalar
  )
  n32 = Multiply(A=Add(A=Time() {period=1}, B=Frac(None=@n31))) {constB=0.04}
  n33 = TextureSample(UVs=ComponentMask(None=VertexInterpolator(VS=@n32).PS) {g=true, r=true}) {texture=&/T_Gameplay_RadioTower_Radar_Random_D}
  n34 = Multiply(A=Add(A=@n33.RGB) {constB=-0.5}) {constB=0.5}
  n35 = Abs(None=Add(A=@n30, B=ComponentMask(None=@n34) {r=true}) {constB=0.1})
  n36 = Max(A=ComponentMask(None=@n35) {r=true}, B=ComponentMask(None=@n35) {g=true})
  n37 = Abs(None=Add(A=@n30, B=ComponentMask(None=@n34) {g=true}) {constB=0.1})
  n38 = Max(A=ComponentMask(None=@n37) {r=true}, B=ComponentMask(None=@n37) {g=true})
  n39 = Abs(None=Add(A=@n30, B=ComponentMask(None=@n34) {b=true}) {constB=0.1})
  n40 = Max(A=ComponentMask(None=@n39) {r=true}, B=ComponentMask(None=@n39) {g=true})
  MaterialExpressionAppend3Vector_0 = MaterialXAppend3Vector(A=OneMinus(None=Round(None=@n36)), B=OneMinus(None=Round(None=@n38)), C=OneMinus(None=Round(None=@n40)))
  n41 = DotProduct(A=@MaterialExpressionAppend3Vector_0, B=Constant3Vector() {constant=<rgba 0.33 0.33 0.33 0>})
  Multiply(A=@n41, B=@n33.A)
  n42 = Constant3Vector() {constant=<rgba 0.2031 2.0791 3 0>}
  n43 = Constant3Vector() {constant=<rgba 0.209 3 0 1>}
  n44 = ScalarParameter() {bUseCustomPrimitiveData=true, parameterName=Activation Complete, primitiveDataIndex=2}
  Subtract(A=LinearInterpolate(A=Multiply(A=@MaterialExpressionAppend3Vector_0) {constB=5}, B=LinearInterpolate(A=@n42, B=@n43, Alpha=@n44), Alpha=Round(None=@n41))) {constB=0.02}
 ]
}
